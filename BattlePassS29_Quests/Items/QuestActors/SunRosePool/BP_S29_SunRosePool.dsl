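; Season 29 quest scrying pool that shows an Oracle body or character visions during NPC conversations.
; For the local player, conversation node parameters select Oracle spawn and emotes, or a character and montage for the child-actor mannequin.
; Once mannequin materials load, they become water-style MIDs keeping their normal maps, and the vision replaces the Oracle with Niagara effects.
; Conversation state changes drive despawn and ripple fades; low-end platforms use low-detail materials.
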(event EventBeginPlay
  (|CallParentFunction)
  (CallFunction|MakeMIDPoolWater))

(event Collision|EventActorBeginOverlap
  (params (OtherActor :type Actor%20Object%20Reference))
  (|CallParentFunction OtherActor) ; This node is disabled and will not be called.\nDrag off pins to build functionality.
)

(event EventTick
  (params (DeltaSeconds :type Float%20%28single-precision%29))
  (|CallParentFunction DeltaSeconds) ; This node is disabled and will not be called.\nDrag off pins to build functionality.
)

(event OnReceivedLocalClientConversationMessage(NonPlayerConversationComponent)
  (params (LocalPlayerController :type Fort%20Player%20Controller%20Object%20Reference) (Message_Message_SpeakerID :type Gameplay%20Tag%20Structure) (Message_Message_ParticipantDisplayName :type Text) (Message_Message_Text :type Text) (Message_Message_MetadataParameters :type Array%20of%20Conversation%20Node%20Parameter%20Pair%20Structures) (Message_Participants_List :type Array%20of%20Conversation%20Participant%20Entry%20Structures) (Message_CurrentNode :type Conversation%20Node%20Handle%20Structure) (Message_Options :type Array%20of%20Client%20Conversation%20Option%20Entry%20Structures))
  (bind _output (Utilities|Array|Get(aref) Message_Message_MetadataParameters))
  (bind (_name _value) (Utilities|Struct|BreakConversationNodeParameterPair _output))
  (bind _mannequin_actor (Variables|Default|GetMannequin%20Actor))
  (bind _childactor (Variables|/Script/Engine.ChildActorComponent|GetChildActor _mannequin_actor))
  (Utilities|FlowControl|Sequence
    (:then_0
      ;[1 1 1] Enable Water Ripple animation
      (CallFunction|WaterRippleFadeIn false))
    (:then_1
      ;[1 1 1] Checking to see if there's a parameter on a conversation node
      (bind _output_2 (CallFunction|IsValidLocalPlayerController LocalPlayerController))
      (if (and _output_2 (!= (Utilities|Array|Length Message_Message_MetadataParameters) 0))
        (if (not (Utilities|String|IsEmpty _name))
          (Variables|Default|SetNew%20Character%20to%20Display 0)
          (Variables|Default|SetAnimation%20Montage 0)
          ;[1 0.5938 0.939] Based on Convo node choice, set character and their idle montage.
          (switch string _name
            (:Aphrodite
              (Variables|Default|SetNew%20Character%20to%20Display "/BRCosmetics/Athena/Items/Cosmetics/Characters/Character_LiftingAura_Wave.Character_LiftingAura_Wave")
              (Variables|Default|SetAnimation%20Montage "/BRCosmetics/Animation/Game/MainPlayer/Menu/BR/CustomSkins/_S29/LiftingAura_WaveF/LiftingAura_Wave_Idle_Montage.LiftingAura_Wave_Idle_Montage"))
            (:Ares
              (Variables|Default|SetNew%20Character%20to%20Display "/BRCosmetics/Athena/Items/Cosmetics/Characters/Character_GoldAccomplishment.Character_GoldAccomplishment")
              (Variables|Default|SetAnimation%20Montage "/Game/Animation/Game/MainPlayer/Menu/Male/Medium/Commando/Commando_Male_Body_Montage_Idle.Commando_Male_Body_Montage_Idle"))
            (:Marigold
              (Variables|Default|SetNew%20Character%20to%20Display "/BRCosmetics/Athena/Items/Cosmetics/Characters/CID_A_034_Athena_Commando_F_CatBurglar.CID_A_034_Athena_Commando_F_CatBurglar")
              ;;[1 1 1 1 10] Marigold uses a default idle animation
              (Variables|Default|SetAnimation%20Montage "/Game/Animation/Game/MainPlayer/Menu/Female/Medium/Commando/Commando_Female_Body_Montage_Idle.Commando_Female_Body_Montage_Idle"))
            (:Brutus
              (Variables|Default|SetNew%20Character%20to%20Display "/BRCosmetics/Athena/Items/Cosmetics/Characters/CID_692_Athena_Commando_M_HenchmanTough.CID_692_Athena_Commando_M_HenchmanTough")
              (Variables|Default|SetAnimation%20Montage "/BRCosmetics/Animation/Game/MainPlayer/Menu/BR/CustomSkins/_S12/Henchman/Henchman_Male_Idle_Montage.Henchman_Male_Idle_Montage"))
            (:Jules
              (Variables|Default|SetNew%20Character%20to%20Display "/BRCosmetics/Athena/Items/Cosmetics/Characters/CID_770_Athena_Commando_F_MechanicalEngineer.CID_770_Athena_Commando_F_MechanicalEngineer")
              (Variables|Default|SetAnimation%20Montage "/BRCosmetics/Animation/Game/MainPlayer/Menu/BR/CustomSkins/_s13/MechanicalEngineer/MechanicalEngineer_Female_Idle_Montage.MechanicalEngineer_Female_Idle_Montage"))
            (:Cerberus
              (Variables|Default|SetNew%20Character%20to%20Display "/BRCosmetics/Athena/Items/Cosmetics/Characters/Character_GateHound_Spectral.Character_GateHound_Spectral")
              (Variables|Default|SetAnimation%20Montage "/BRCosmetics/Animation/Game/MainPlayer/Menu/BR/CustomSkins/_S29/GateHoundSpectral/GateHoundSpectral_Male_idle_Montage.GateHoundSpectral_Male_idle_Montage"))
            (:Hades
              (Variables|Default|SetNew%20Character%20to%20Display "/BRCosmetics/Athena/Items/Cosmetics/Characters/Character_DarkStance_Inferno.Character_DarkStance_Inferno")
              (Variables|Default|SetAnimation%20Montage "/Game/Animation/Game/MainPlayer/Menu/Male/Medium/Commando/Commando_Male_Body_Montage_Idle.Commando_Male_Body_Montage_Idle"))
            (:Medusa
              (Variables|Default|SetNew%20Character%20to%20Display "/BRCosmetics/Athena/Items/Cosmetics/Characters/Character_ReptilianOcean_Sleek.Character_ReptilianOcean_Sleek")
              (Variables|Default|SetAnimation%20Montage "/Game/Animation/Game/MainPlayer/Menu/Female/Medium/Commando/Commando_Female_Body_Montage_Idle.Commando_Female_Body_Montage_Idle"))
            (:OracleEmote
              ;[1 0.817 0] For VFX and SFX
              (CallFunction|OracleEmote _value))
            (:Oracle%20Convo%20Start
              ;[0 1 0.9494] Remove the mannequin on conversation start and spawn Oracle
              (if (Rendering|IsVisible (Variables|Default|GetMannequin%20Actor))
                (CallFunction|MannequinVisionDisappears)
                (else
                  _))
              (CallFunction|SpawnOracle)))
          (Utilities|IsValid _childactor
            (:Is%20Valid
              (bind _asbp_s29_sun_rose_pool_mannequin (Utilities|Casting|CastToBP_S29_SunRosePool_Mannequin _childactor)
                (:then
                  (Variables|Default|SetAs%20BP%20S29%20Sun%20Rose%20Pool%20Mannequin _asbp_s29_sun_rose_pool_mannequin)
                  (Class|BPS29SunRosePoolMannequin|ChangeMannequin _asbp_s29_sun_rose_pool_mannequin (Variables|Default|GetNew%20Character%20to%20Display))
                  ;[1 1 1] Wait for all the skeletal mesh and materials to load before changing their material and assigning the montage
                  (EventDispatchers|BindEventtoOnMaterialsUpdated _asbp_s29_sun_rose_pool_mannequin (EventDispatchers|CreateEvent :!event Custom|UpdateMatsandPlayMontage)))
                (:CastFailed
                  _)))
            (:Is%20Not%20Valid
              _))
          (Development|PrintString "Hehe can't find me"))))))

(event Custom|UpdateMatsandPlayMontage
  (params (Materials :type Array%20of%20Material%20Instance%20Dynamic%20Object%20References))
  (bind _waterplane (Variables|Default|GetWaterPlane))
  (bind _returnvalue (Transformation|GetWorldLocation _waterplane))
  ;[1 1 1] Wait for all the skeletal mesh and materials to load before changing their material and assigning the montage
  (bind _skeletalmeshcomponents (Customization|GetAllPartSkeletalMeshComponents (stale OnReceivedLocalClientConversationMessage(NonPlayerConversationComponent) _asbp_s29_sun_rose_pool_mannequin)))
  (for _array_element _skeletalmeshcomponents
    ;[1 1 1] Make Material for Character
    (for (_array_element_2 _array_index) (Rendering|Material|GetMaterials _array_element)
      (bind _returnvalue_2 (Rendering|Material|CreateDynamicMaterialInstance _array_element _array_index _array_element_2 "" :!class /Script/Engine.PrimitiveComponent))
      (Utilities|IsValid _returnvalue_2
        (:Is%20Valid
          (bind _returnvalue_3 (Rendering|Material|GetTextureParameterValue _returnvalue_2 "Normals" :!class /Script/Engine.MaterialInstanceDynamic))
          (bind _returnvalue_4 (Rendering|Material|CreateDynamicMaterialInstance _array_element _array_index (Variables|Default|GetBaseMID) "" :!class /Script/Engine.PrimitiveComponent))
          (Utilities|Array|Add (Variables|VFX|GetMIDs_Character) _returnvalue_4)
          (Utilities|IsValid _returnvalue_3
            (:Is%20Valid
              (Rendering|Material|SetTextureParameterValue _returnvalue_4 "Normals" _returnvalue_3))
            (:Is%20Not%20Valid
              (Rendering|Material|SetTextureParameterValue _returnvalue_4 "Normals" "/Engine/EngineMaterials/BaseFlattenNormalMap.BaseFlattenNormalMap")))
          (Rendering|Material|SetScalarParameterValue _returnvalue_4 "WordPosition_Z" (.z _returnvalue) :!class /Script/Engine.MaterialInstanceDynamic)
          (Rendering|Material|SetScalarParameterValue _returnvalue_4 "GameTime" (Utilities|Time|GetGameTimeinSeconds) :!class /Script/Engine.MaterialInstanceDynamic)
          (Rendering|Material|SetScalarParameterValue _returnvalue_4 "Fadeout" 0.0 :!class /Script/Engine.MaterialInstanceDynamic)
          (Rendering|Material|SetScalarParameterValue _returnvalue_4 "BlendToRadial" 1.0 :!class /Script/Engine.MaterialInstanceDynamic)
          (Rendering|Material|SetVectorParameterValue _returnvalue_4 "ObjectPosition" (Math|Conversions|ToLinearColor(Vector) _returnvalue) :!class /Script/Engine.MaterialInstanceDynamic)))))
  (Class|BPS29SunRosePoolMannequin|LoadandPlayCurrentMontage (stale OnReceivedLocalClientConversationMessage(NonPlayerConversationComponent) _asbp_s29_sun_rose_pool_mannequin) (Variables|Default|GetAnimation%20Montage) 0.0 false)
  (CallFunction|DespawnOracle)
  (CallFunction|MannequinVisionAppears))

(event OnDialogMarkerInteractionStateChanged(NonPlayerConversationComponent)
  (params (CurrentInteractionState :type EDialogMarkerInteractionState%20Enum :default "Conversation") (InteractingController :type Fort%20Player%20Controller%20Object%20Reference))
  ;[1 1 1] Turn off the mannequin when the player leaves the conversation
  (switch Utilities|FlowControl|Switch|SwitchonEDialogMarkerInteractionState CurrentInteractionState
    (:Conversation
      (Variables|Default|SetInConversation true))
    (:InteractionRange
      _)
    (:IndicatorRange
      _)
    (:None
      _))
  (bind _output (CallFunction|IsValidLocalPlayerController InteractingController))
  (if (and _output (Variables|Default|GetInConversation))
    (Variables|Default|SetInConversation false)
    (if (Variables|VFX|GetFX_IsOracleSpawned)
      (CallFunction|DespawnOracle)
      (else
        (CallFunction|MannequinVisionDisappears)))
    (CallFunction|WaterRippleFadeIn true)))

(event Custom|MannequinVisionAppears
  (bind _mannequin_actor (Variables|Default|GetMannequin%20Actor))
  (bind (_returnvalue_location _returnvalue_rotation) (Math|Transform|BreakTransform (Transformation|GetWorldTransform _mannequin_actor)))
  (bind (_returnvalue_roll _returnvalue_pitch _returnvalue_yaw) (Math|Rotator|BreakRotator (Transformation|GetActorRotation (Game|GetPlayerCharacter 0))))
  (bind _mannequin_actor_2 (Variables|Default|GetMannequin%20Actor))
  ;[1 0.817 0] For VFX and SFX
  ;;[1 1 1] Spawn Mannequin facing the player
  (Transformation|SetWorldRotation _mannequin_actor_2 (Math|Rotator|MakeRotator 0.0 0.0 (+ _returnvalue_yaw 9e+01)) false false)
  (Rendering|SetVisibility _mannequin_actor_2 true false)
  (Utilities|FlowControl|Sequence
    (:then_0
      (bind _fx_activevisionidle (Variables|Default|GetFX_ActiveVisionIdle)
        (:then
          (Components|DestroyComponent _fx_activevisionidle))
        (:else)))
    (:then_1
      (bind _returnvalue (Niagara|SpawnSystematLocation (Variables|VFX|GetFX_OnSpawnVision) _returnvalue_location _returnvalue_rotation))
      (CallFunction|SetVisionFXParams _returnvalue)
      (bind _returnvalue_2 (Niagara|SpawnSystematLocation (Variables|VFX|GetFX_VisionIdle) _returnvalue_location _returnvalue_rotation))
      (bind _output_get (Variables|Default|SetFX_ActiveVisionIdle _returnvalue_2))
      (CallFunction|SetVisionFXParams _output_get))))

(event Custom|SpawnOracle
  (bind _ns_s29_oracelbodyspawn (Variables|Default|GetNS_S29_OracelBodySpawn))
  ;[1 0.817 0] For VFX and SFX
  (bind _output_get (Variables|VFX|SetFX_OracleNextShapeInt 0))
  (CallFunction|SetOracleBodyParams _output_get true)
  (if (not (Variables|VFX|GetFX_IsOracleSpawned))
    (Components|Activation|Activate _ns_s29_oracelbodyspawn true)
    (Niagara|SetNiagaraVariable(Bool) _ns_s29_oracelbodyspawn "IsSpawning" true)
    (Audio|SpawnSoundAttached "/BattlePassS29_Quests/Audio/SFX/Quests_S29_OraclePool_Activate_Cue.Quests_S29_OraclePool_Activate_Cue" _ns_s29_oracelbodyspawn)
    (bind _returnvalue (Audio|SpawnSoundAttached "/BattlePassS29_Quests/Audio/SFX/Quests_S29_OraclePool_Active_Loop_Cue.Quests_S29_OraclePool_Active_Loop_Cue" _ns_s29_oracelbodyspawn))
    (Variables|VFX|SetFX_IsOracleSpawned true)))

(event Custom|MannequinVisionDisappears
  (bind (_returnvalue_location _returnvalue_rotation) (Math|Transform|BreakTransform (Math|Transform|ComposeTransforms (struct :Location "0.000000,0.000000,90.000000" :Rotation "0, 0, 0" :Scale "1.000000,1.000000,1.000000") (Transformation|GetActorTransform))))
  (bind _as_bp_s29_sun_rose_pool_mannequin (Variables|Default|GetAs%20BP%20S29%20Sun%20Rose%20Pool%20Mannequin))
  ;[1 0.817 0] For VFX and SFX
  (EventDispatchers|UnbindEventfromOnMaterialsUpdated (Variables|Default|GetAs%20BP%20S29%20Sun%20Rose%20Pool%20Mannequin) (EventDispatchers|CreateEvent :!event Custom|UpdateMatsandPlayMontage))
  (Utilities|FlowControl|Sequence
    (:then_0
      (for _array_element (Variables|VFX|GetMIDs_Character)
        (Rendering|Material|SetScalarParameterValue _array_element "GameTime" (Utilities|Time|GetGameTimeinSeconds) :!class /Script/Engine.MaterialInstanceDynamic)
        (Rendering|Material|SetScalarParameterValue _array_element "Fadeout" 1.0 :!class /Script/Engine.MaterialInstanceDynamic))
      (bind _returnvalue (Niagara|SpawnSystematLocation (Variables|VFX|GetFX_OnRemoveVision) _returnvalue_location _returnvalue_rotation))
      (CallFunction|SetVisionFXParams _returnvalue)
      (bind _fx_activevisionidle (Variables|Default|GetFX_ActiveVisionIdle)
        (:then
          (Components|DestroyComponent _fx_activevisionidle))
        (:else)))
    (:then_1
      (Utilities|FlowControl|Delay 2.0)
      (Utilities|IsValid _as_bp_s29_sun_rose_pool_mannequin
        (:Is%20Valid
          (Class|BPS29SunRosePoolMannequin|StopMontage _as_bp_s29_sun_rose_pool_mannequin)
          (Class|BPS29SunRosePoolMannequin|TurnOffMannequin _as_bp_s29_sun_rose_pool_mannequin 0)
          (Development|PrintString "Show's over, go home"))
        (:Is%20Not%20Valid
          _))
      (Rendering|SetVisibility (Variables|Default|GetMannequin%20Actor) false false))))

(event Custom|DespawnOracle
  (bind _ns_s29_oraclebody_idle (Variables|Default|GetNS_S29_OracleBody_Idle))
  (bind _returnvalue (Transformation|GetWorldLocation _ns_s29_oraclebody_idle))
  (bind _ns_s29_oraclebody_idle_2 (Variables|Default|GetNS_S29_OracleBody_Idle))
  ;[1 0.817 0] For VFX and SFX
  (Utilities|IsValid _ns_s29_oraclebody_idle_2
    (:Is%20Valid
      (Niagara|SpawnSystematLocation "/BattlePassS29_Quests/Items/FX/OracleBody/NS_S29_OracleBody_Disappear.NS_S29_OracleBody_Disappear" _returnvalue)
      (Audio|SpawnSoundatLocation "/BattlePassS29_Quests/Audio/SFX/Quests_S29_OraclePool_Deactivate_Cue.Quests_S29_OraclePool_Deactivate_Cue" _returnvalue)
      (Audio|SpawnSoundatLocation "/BattlePassS29_Quests/Audio/SFX/Quests_S29_OraclePool_Loop_Cue.Quests_S29_OraclePool_Loop_Cue" _returnvalue)
      (Audio|Components|Audio|Stop (stale Custom|SpawnOracle _returnvalue))
      (Niagara|SetNiagaraVariable(Bool) _ns_s29_oraclebody_idle_2 "IsSpawning" false)
      (Variables|VFX|SetFX_IsOracleSpawned false))))

(event Custom|Oracle%20Emote
  (params (Emote%20Value :type String))
  ;[1 0.817 0] For VFX and SFX
  (switch string Emote%20Value
    (:Idle
      ;;[1 1 1] Set Oracle Emotes Here!
      (bind _output_get (Variables|VFX|SetFX_OracleNextShapeInt 0)))
    (:Think
      (bind _output_get_2 (Variables|VFX|SetFX_OracleNextShapeInt 1))
      (CallFunction|SetOracleBodyParams _output_get_2 true))
    (:Question
      (bind _output_get_3 (Variables|VFX|SetFX_OracleNextShapeInt 2))
      (CallFunction|SetOracleBodyParams _output_get_3 true))
    (:Exclamation
      (bind _output_get_4 (Variables|VFX|SetFX_OracleNextShapeInt 4))
      (CallFunction|SetOracleBodyParams _output_get_4 true))
    (:Awkward
      (bind _output_get_5 (Variables|VFX|SetFX_OracleNextShapeInt 3))
      (CallFunction|SetOracleBodyParams _output_get_5 true))
    (:Glow
      (bind _output_get_6 (Variables|VFX|SetFX_OracleNextShapeInt 0))
      (CallFunction|SetOracleBodyParams _output_get_6 true))
    (:Default
      _))
  (CallFunction|SetOracleBodyParams _output_get true))

(fn IsValidLocalPlayerController
  (params (InstigatingPlayerController :type Fort%20Player%20Controller%20Object%20Reference))
  ;[0 0 0] Returns true on the executing client if the instigator controller is locally controlled. Returns false if running on the server
  (if (not (Networking|IsServer))
    (Utilities|IsValid InstigatingPlayerController
      (:Is%20Valid
        (if (PlayerController|IsinSplitscreen InstigatingPlayerController)
          (for _array_element (Fortnite|GetLocalFortPlayerControllers true)
            (if (== _array_element InstigatingPlayerController)
              (return true)))
          (return false)
          (else
            (return true))))
      (:Is%20Not%20Valid
        _)))
  (return false))

(fn MakeMIDPoolWater
  ;[1 1 1] Make MID for Water Animations
  (bind _returnvalue (Rendering|Material|CreateDynamicMaterialInstance (Variables|Default|GetWaterPlane) 0 (select (Math|Boolean|ORBoolean (TouchPlatform|IsMobileGame) (Fortnite|IsMobilePlatform) (Class|PlatformHelpersLibrary|IsPcPerfMode) (TouchPlatform|IsNintendoSwitchorNintendoSwitchPreview)) "/Game/Environments/Helios/Props/Quest/Quest_ScryingPool_A/Materials/M_ScryingPoolWater_LowDetail.M_ScryingPoolWater_LowDetail" "/Game/Environments/Helios/Props/Quest/Quest_ScryingPool_A/Materials/M_ScryingPoolWater.M_ScryingPoolWater") "None" :!class /Script/Engine.PrimitiveComponent))
  (bind _output_get (Variables|VFX|SetMID_Water _returnvalue))
  (Rendering|Material|SetScalarParameterValue _returnvalue "BlendToRadial" 0.0 :!class /Script/Engine.MaterialInstanceDynamic))

(fn SetOracleBodyParams
  (params (ShapeIndex :type Integer) (IsGlowing :type Boolean))
  (bind _ns_s29_oraclebody_idle (Variables|Default|GetNS_S29_OracleBody_Idle))
  (if (!= (Variables|VFX|GetFX_OracleNextShapeInt) (Variables|VFX|GetFX_OracleShapeInt))
    (Niagara|SpawnSystematLocation "/BattlePassS29_Quests/Items/FX/OracleBody/NS_S29_OracleBody_ShapeSwitch.NS_S29_OracleBody_ShapeSwitch" (Transformation|GetWorldLocation (Variables|Default|GetNS_S29_OracleBody_Idle))))
  (Niagara|SetNiagaraVariable(Int32) _ns_s29_oraclebody_idle "NextShapeIndex" ShapeIndex)
  (Niagara|SetNiagaraVariable(Bool) _ns_s29_oraclebody_idle "IsGlowing" IsGlowing)
  (Variables|VFX|SetFX_OracleShapeInt ShapeIndex))

(fn SetVisionFXParams
  (params (NiagaraSystem :type Niagara%20Particle%20System%20Component%20Object%20Reference))
  (Utilities|FlowControl|Sequence
    (:then_0
      (bind _asskeletal_mesh_component (Utilities|Casting|CastToSkeletalMeshComponent (Variables|/Script/Engine.Actor|GetRootComponent (Variables|/Script/Engine.ChildActorComponent|GetChildActor (Variables|Default|GetMannequin%20Actor)))))
      (Niagara|SetNiagaraSkeletalMeshComponent NiagaraSystem "SourceMesh" _asskeletal_mesh_component))))

(fn ConstructionScript
  (|CallParentFunction)
  (CallFunction|MakeMIDPoolWater)
  (bind _returnvalue (Rendering|Material|CreateDynamicMaterialInstance (select (Math|Boolean|ORBoolean (TouchPlatform|IsMobileGame) (Fortnite|IsMobilePlatform) (Class|PlatformHelpersLibrary|IsPcPerfMode) (TouchPlatform|IsNintendoSwitchorNintendoSwitchPreview)) "/Game/Environments/Helios/Props/Quest/Quest_ScryingPool_A/Materials/M_ScryingPoolWater_Character_LowDetail.M_ScryingPoolWater_Character_LowDetail" "/Game/Environments/Helios/Props/Quest/Quest_ScryingPool_A/Materials/M_ScryingPoolWater_Character.M_ScryingPoolWater_Character") "None" "None" :!class /Script/Engine.KismetMaterialLibrary))
  (bind _output_get (Variables|Default|SetBaseMID _returnvalue))
  (Rendering|Material|SetTextureParameterValue _output_get "Normals" "/Engine/EngineMaterials/BaseFlattenNormalMap.BaseFlattenNormalMap"))

(fn WaterRippleFadeIn
  (params (Fadeout :type Boolean))
  (bind _returnvalue (Math|Float|SelectFloat 0.0 1.0 Fadeout))
  (bind _waterrippleactive (Variables|VFX|GetWaterRippleActive))
  (if Fadeout
    (if _waterrippleactive
      (Variables|VFX|SetWaterRippleActive false))
    (elif _waterrippleactive
      _
      (else
        (Variables|VFX|SetWaterRippleActive true))))
  (Rendering|Material|SetScalarParameterValue (Variables|VFX|GetMID_Water) "GameTime_Ripple" (Utilities|Time|GetGameTimeinSeconds) :!class /Script/Engine.MaterialInstanceDynamic)
  (Rendering|Material|SetScalarParameterValue (Variables|VFX|GetMID_Water) "BlendToRadial" _returnvalue :!class /Script/Engine.MaterialInstanceDynamic)
  (for _array_element (Variables|VFX|GetMIDs_Character)
    (Rendering|Material|SetScalarParameterValue _array_element "GameTime_Ripple" (Utilities|Time|GetGameTimeinSeconds) :!class /Script/Engine.MaterialInstanceDynamic)
    (Rendering|Material|SetScalarParameterValue _array_element "BlendToRadial" _returnvalue :!class /Script/Engine.MaterialInstanceDynamic)))
