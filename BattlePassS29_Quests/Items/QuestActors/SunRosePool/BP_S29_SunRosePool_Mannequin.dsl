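; Battle Pass S29 quest mannequin actor for the Sun Rose Pool that displays Athena characters and animation montages on its skeletal mesh.
; Change and TurnOff events both apply a passed character definition with stored override variants.
; LoadAndPlayCurrentMontage async-loads a soft montage, plays it looping from a given position and can set a timer for PlayIdleMontage.
; SetIdleMontage and PlayIdleMontage are empty here, so that timer triggers no logic in this graph.
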(event EventBeginPlay
  ;[1 1 1] Begin Play
  (|CallParentFunction))

(event Custom|StopMontage
  ;[1 1 1] Stop Montage
  (Components|Animation|Stop (Variables|SkeletalMeshActor|GetSkeletalMeshComponent)))

(event Custom|SetIdleMontage
  (params (Animation%20Montage :type Anim%20Montage%20Soft%20Object%20Reference)))

(event Custom|PlayIdleMontage)

(event Custom|LoadAndPlayCurrentMontage
  (params (Animation%20Montage :type Anim%20Montage%20Soft%20Object%20Reference) (AnimCurrentPosition :type Float%20%28double-precision%29) (ReturnToIdle :type Boolean))
  ;[1 1 1] Load and Play Current Montage
  (Variables|Default|SetReturn%20to%20Idle ReturnToIdle)
  (bind _object (Utilities|AsyncLoadAsset Animation%20Montage)
    (:Completed
      (bind _asanim_montage (Utilities|Casting|CastToAnimMontage _object))
      (Utilities|IsValid _asanim_montage
        (:Is%20Valid
          (Variables|Default|SetCurrentMontage _asanim_montage)
          (Variables|Default|SetMontageStartingPosition AnimCurrentPosition)
          (bind _skeletalmeshcomponent (Variables|SkeletalMeshActor|GetSkeletalMeshComponent)
            (:then
              (Components|Animation|PlayAnimation _skeletalmeshcomponent (Variables|Default|GetCurrentMontage) true)
              (Components|Animation|SetPosition (Variables|SkeletalMeshActor|GetSkeletalMeshComponent) (Variables|Default|GetMontageStartingPosition) true)
              (if (Variables|Default|GetReturn%20to%20Idle)
                (bind _returnvalue (Utilities|Time|SetTimerbyEvent (EventDispatchers|CreateEvent :!event Custom|PlayIdleMontage) (- (Animation|GetPlayLength (Variables|Default|GetCurrentMontage)) (Variables|Default|GetMontageStartingPosition))))
                (Variables|Default|SetTimerHandle_PlayIdleMontage _returnvalue)))
            (:else)))))))

(event Custom|Change%20Mannequin
  (params (NewCharacterToDisplay :type Athena%20Character%20Item%20Definition%20Object%20Reference))
  ;[1 1 1] Change Mannequin to Display
  ;;[1 1 1] Set Character to Display
  (Customization|SetAthenaCharacterDefandVariants NewCharacterToDisplay (Variables|Default|GetIn%20Override%20Variants)))

(event Custom|TurnOffMannequin
  (params (Clear%20Character :type Athena%20Character%20Item%20Definition%20Object%20Reference))
  ;[1 1 1] Change Mannequin to Display
  (Customization|SetAthenaCharacterDefandVariants Clear%20Character (Variables|Default|GetIn%20Override%20Variants)))

(fn ConstructionScript
  (|CallParentFunction))
